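; Burst gameplay cue for a player bouncing on a tire, adding bounce animation and an audio indicator to the parent cue's effects.
; When the effect causer is valid, it plays a hit bounce on the target building mesh, driven by the cue's magnitude, normal and location.
; It also shows a short-range footstep sound indicator, using the causer's team with hostile and neutral affiliations.
; The Niagara burst override only calls its parent.
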(fn OnBurst(EitherVFX-preferthisversion)
  (params (MyTarget :type Actor%20Object%20Reference) (Parameters :type Gameplay%20Cue%20Parameters%20Structure%20%28by%20ref%29) (ParticleComponents :type Array%20of%20FXSystem%20Component%20Object%20References) (AudioComponents :type Array%20of%20Audio%20Component%20Object%20References) (BurstCameraShakeInstance :type Camera%20Shake%20Base%20Object%20Reference) (BurstDecalInstance :type Decal%20Actor%20Object%20Reference))
  (bind (_normalizedmagnitude _rawmagnitude _effectcontext _matchedtagname _originaltag _aggregatedsourcetags _aggregatedtargettags _location _normal _instigator _effectcauser) (Ability|GameplayCue|BreakGameplayCueParameters Parameters))
  (|CallParentFunction MyTarget Parameters ParticleComponents AudioComponents BurstCameraShakeInstance BurstDecalInstance)
  (bind _asbuilding_static_mesh (Utilities|Casting|CastToBuildingSMActor MyTarget))
  (Utilities|IsValid _effectcauser
    (:Is%20Valid
      (Class|BuildingStaticMesh|StartBounceAnimation _asbuilding_static_mesh (Utilities|Time|GetTimeSeconds MyTarget) _normalizedmagnitude _normal _location "Hit" false)
      (Fortnite|ShowSoundIndicatorLocalforTeam :WorldContextObject MyTarget :Location _location :MaxAudibleDistance 7e+02 :IndicatorType "FootStep" :Team (Team|GetActorTeam _effectcauser) :Affiliations (Utilities|Array|MakeArray "Hostile" "Neutral") :Instigator MyTarget :Tint "(R=0.000000,G=0.000000,B=0.000000,A=0.000000)" :AttenuationMaxRangeMultiplier 1.0 :AttenuationIntensityMultiplier 1.0)
      (return))))

(fn OnBurst(Niagara)
  (params (MyTarget :type Actor%20Object%20Reference) (Parameters :type Gameplay%20Cue%20Parameters%20Structure%20%28by%20ref%29) (ParticleComponents :type Array%20of%20Niagara%20Particle%20System%20Component%20Object%20References) (AudioComponents :type Array%20of%20Audio%20Component%20Object%20References) (BurstCameraShakeInstance :type Camera%20Shake%20Base%20Object%20Reference) (BurstDecalInstance :type Decal%20Actor%20Object%20Reference))
  (|CallParentFunction MyTarget Parameters ParticleComponents AudioComponents BurstCameraShakeInstance BurstDecalInstance)
  (return))
